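; Sparks map controller that activates and unloads World Partition data layers from data tables.
; At BeginPlay every instance derives a crowd-detail tier from platform, applies that scalability row's playlist-matched layers and unloads other tiers'.
; Authority also applies playlist-tag layers and activation-CVAR game-mode layers, falling back to the default mode if unmatched.
; Activation changes reapply game-mode layers from the comma-separated string, without playlist checks.
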(event EventBeginPlay
  ;[1 1 1] Use Debug?
  (Utilities|FlowControl|Sequence
    (:then_0
      (|CallParentFunction)
      (CallFunction|CheckPlatform)
      (CallFunction|ClientSideDL)
      (Networking|SwitchHasAuthority
        (:Authority
          (CallFunction|ServersideDL))))
    (:then_1
      (if (Editor|InEditororPIE)
        ;[1 0.05035 0.05035] Debug
        (Utilities|FlowControl|Sequence
          (:then_0
            ;;[1 1 1] Print Playlist
            (bind _gamestate (Class|FortAsyncActionCurrentPlaylistReady|PlaylistDataReadyAsync)
              (:OnReady
                (Development|PrintString (Utilities|String|Append "Active Sparks Playlist: " (Utilities|String|ToString(Name) (Playlist|GetCurrentPlaylistName _gamestate))))
                ;;[1 1 1] Get current Activation
                (Development|PrintString (Utilities|String|Append "Active Sparks activation (Cvar): " (CallFunction|GetCurrentSparksActivation)))
                ;;[1 1 1] Get Scalability
                (Development|PrintString (Utilities|String|Append "Is low power device? " (Utilities|String|ToString(Boolean) (or (TouchPlatform|IsNintendoSwitch) (TouchPlatform|IsMobileGame))))))))
          (:then_1
            ;;[1 1 1] Print Playlist Data table info
            (bind _playlists (CallFunction|GetPlaylistData))
            (for _array_element _playlists
              (Development|PrintString (Utilities|String|Append "This is the playlist from the data table:" (GameplayTags|GetDebugStringfromGameplayTag (Utilities|Struct|BreakSparksPlaylistDataLayerData _array_element)))))))))))

(event Custom|ServersideDL
  (bind _returnvalue (CallFunction|GetCurrentSparksActivation))
  (Utilities|FlowControl|Sequence
    (:then_0
      ;[1 0.5691 0.5857] Playlist  Manager
      (bind _playlists (CallFunction|GetPlaylistData))
      (for _array_element _playlists
        (if (Class|FortGameStateAthena|PlaylistHasTag (Fortnite|Athena|GameState|GetGameStateAthena) (Utilities|Struct|BreakSparksPlaylistDataLayerData _array_element))
          (Utilities|FlowControl|Sequence
            (:then_0
              (for _array_element_2 (.EnabledDataLayers_9_41BB9A3E4648805B5AF70AA728288C92 (Utilities|Struct|BreakSparksPlaylistDataLayerData _array_element))
                (CallFunction|DatalayerManager _array_element_2 "Activated" true)))
            (:then_1
              (for _array_element_3 (.DisabledDataLayers_10_8DE59E3B474C88ED2E7988B185572444 (Utilities|Struct|BreakSparksPlaylistDataLayerData _array_element))
                (CallFunction|DatalayerManager _array_element_3 "Unloaded" true)))))))
    (:then_1
      ;[0.9247 1 0.7019] CVAR Manager
      (bind _gamemode (CallFunction|GetModeData))
      (for _array_element_4 _gamemode
        (if (Utilities|String|EqualExactly(String) _returnvalue (Utilities|Struct|BreakSparksGameModeDataLayerData _array_element_4))
          (for _array_element_5 (.PlaylistTags_29_5C7DCEE44725600F2FDC8BA54608C35D (Utilities|Struct|BreakSparksGameModeDataLayerData _array_element_4))
            (if (Class|FortGameStateAthena|PlaylistHasTag (Fortnite|Athena|GameState|GetGameStateAthena) _array_element_5)
              (Variables|Default|SetCvarFound true)
              (Fortnite|FortShippingLog self (Utilities|String|Append "CVAR:" _returnvalue " Enabled, Using Playlists: " ""))
              (Utilities|FlowControl|Sequence
                (:then_0
                  (for _array_element_6 (.EnabledDataLayers_9_41BB9A3E4648805B5AF70AA728288C92 (Utilities|Struct|BreakSparksGameModeDataLayerData _array_element_4))
                    (CallFunction|DatalayerManager _array_element_6 "Activated" true)))
                (:then_1
                  (for _array_element_7 (.DisabledDataLayers_10_8DE59E3B474C88ED2E7988B185572444 (Utilities|Struct|BreakSparksGameModeDataLayerData _array_element_4))
                    (CallFunction|DatalayerManager _array_element_7 "Unloaded" true))))))))
      (if (not (Variables|Default|GetCvarFound))
        ;[1 0.507 0.1898] Fallback if CVAR doesnt match
        (Fortnite|FortShippingLog self (Utilities|String|Append _returnvalue ", CVAR is not valid, falling back to default JamClassic"))
        (bind _gamemode_2 (CallFunction|GetModeData))
        (for _array_element_8 _gamemode_2
          (if (Utilities|String|EqualExactly(String) (Utilities|Struct|BreakSparksGameModeDataLayerData _array_element_8) (Variables|Default|GetDefaultGameMode))
            (Utilities|FlowControl|Sequence
              (:then_0
                (for _array_element_9 (.EnabledDataLayers_9_41BB9A3E4648805B5AF70AA728288C92 (Utilities|Struct|BreakSparksGameModeDataLayerData _array_element_8))
                  (CallFunction|DatalayerManager _array_element_9 "Activated" true)))
              (:then_1
                (for _array_element_10 (.DisabledDataLayers_10_8DE59E3B474C88ED2E7988B185572444 (Utilities|Struct|BreakSparksGameModeDataLayerData _array_element_8))
                  (CallFunction|DatalayerManager _array_element_10 "Unloaded" true))))))))))

(event Custom|ClientSideDL
  (bind _crowddetail (Variables|Default|GetCrowdDetail))
  ;[0.5981 0.6634 1] Scalability Based DL Manager for weak platforms
  (bind _outrownames (DataTable|GetDataTableRowNames "/SparksMaps/Environment/BP/DataTables/DT_Sparks_DL_Scalability_.DT_Sparks_DL_Scalability_"))
  (Utilities|FlowControl|Sequence
    (:then_0
      (bind _output_get (Variables|Default|SetDTRows _outrownames))
      (bind _result (ArrayMacros|SafeArrayGet _output_get _crowddetail)
        (:%20%20
          (bind _returnvalue (Utilities|GetDataTableRow "/SparksMaps/Environment/BP/DataTables/DT_Sparks_DL_Scalability_.DT_Sparks_DL_Scalability_" _result))
          (for _array_element (Utilities|Struct|BreakSparksPlaylistDatalayers _returnvalue)
            (if (Class|FortGameStateAthena|PlaylistHasTag (Fortnite|Athena|GameState|GetGameStateAthena) (Utilities|Struct|BreakSparksPlaylistDataLayerData _array_element))
              (Utilities|FlowControl|Sequence
                (:then_0
                  (for _array_element_2 (.EnabledDataLayers_9_41BB9A3E4648805B5AF70AA728288C92 (Utilities|Struct|BreakSparksPlaylistDataLayerData _array_element))
                    (CallFunction|DatalayerManager _array_element_2 "Activated" true)))
                (:then_1
                  (for _array_element_3 (.DisabledDataLayers_10_8DE59E3B474C88ED2E7988B185572444 (Utilities|Struct|BreakSparksPlaylistDataLayerData _array_element))
                    (CallFunction|DatalayerManager _array_element_3 "Unloaded" true)))))))))
    (:then_1
      ;; Ensure Data Layers from other Scalability modes are disabled
      (Utilities|Array|RemoveIndex _output_get _crowddetail)
      (for _array_element_4 (Variables|Default|GetDTRows)
        (bind _returnvalue_2 (Utilities|GetDataTableRow "/SparksMaps/Environment/BP/DataTables/DT_Sparks_DL_Scalability_.DT_Sparks_DL_Scalability_" _array_element_4))
        (for _array_element_5 (Utilities|Struct|BreakSparksPlaylistDatalayers _returnvalue_2)
          (for _array_element_6 (.EnabledDataLayers_9_41BB9A3E4648805B5AF70AA728288C92 (Utilities|Struct|BreakSparksPlaylistDataLayerData _array_element_5))
            (CallFunction|DatalayerManager _array_element_6 "Unloaded" true)))))))

(event EventOnSparksActivationChange
  (params (ActivationString :type String))
  ;[1 0.05035 0.05035] Debug
  (Development|PrintString "New activation detected")
  (bind _returnvalue (Utilities|GetDataTableRow "/SparksMaps/Environment/BP/DataTables/DT_Sparks_DL_Gamemode.DT_Sparks_DL_Gamemode" "NewRow"))
  (for _array_element (Utilities|Struct|BreakSparksGameModeDataLayers _returnvalue)
    (for _array_element_2 (Utilities|String|ParseIntoArray ActivationString ",")
      (if (Utilities|String|EqualExactly(String) _array_element_2 (Utilities|Struct|BreakSparksGameModeDataLayerData _array_element))
        (Utilities|FlowControl|Sequence
          (:then_0
            (for _array_element_3 (.EnabledDataLayers_9_41BB9A3E4648805B5AF70AA728288C92 (Utilities|Struct|BreakSparksGameModeDataLayerData _array_element))
              (Development|PrintString (Utilities|String|Append "Enabling this DL: " (Utilities|GetDisplayName _array_element_3)))
              (CallFunction|DatalayerManager _array_element_3 "Activated" true)))
          (:then_1
            (for _array_element_4 (.DisabledDataLayers_10_8DE59E3B474C88ED2E7988B185572444 (Utilities|Struct|BreakSparksGameModeDataLayerData _array_element))
              (Development|PrintString (Utilities|String|Append "Disabling  this DL: " (Utilities|GetDisplayName _array_element_4)))
              (CallFunction|DatalayerManager _array_element_4 "Unloaded" true))))))))

(fn CheckPlatform
  ; Check Platform
  (if (or (TouchPlatform|IsMobileGame) (Class|PlatformHelpersLibrary|IsPcPerfMode))
    (Variables|Default|SetCrowdDetail 0)
    (elif (TouchPlatform|IsNintendoSwitch)
      (Variables|Default|SetCrowdDetail 1)
      (elif (or (Utilities|String|EqualExactly(String) (Game|GetPlatformName) "PS5") (Utilities|String|EqualExactly(String) (Game|GetPlatformName) "XSX"))
        (Variables|Default|SetCrowdDetail 3)
        (elif (Editor|InEditororPIE)
          (Variables|Default|SetCrowdDetail (Rendering|GetRenderingDetailMode))
          (else
            (Variables|Default|SetCrowdDetail 2)))))))

(fn DatalayerManager
  (params (Data%20Layer :type Data%20Layer%20Asset%20Object%20Reference) (State :type EDataLayerRuntimeState%20Enum :default "Unloaded") (Recursive :type Boolean))
  (bind _returnvalue (WorldPartition|GetDataLayerManager))
  (bind _returnvalue_2 (DataLayers|SetDataLayerInstanceRuntimeState _returnvalue (DataLayers|GetDataLayerInstancefromAsset _returnvalue Data%20Layer) State Recursive))
  (return State _returnvalue_2))

(fn GetModeData
  (bind _returnvalue (Utilities|GetDataTableRow "/SparksMaps/Environment/BP/DataTables/DT_Sparks_DL_Gamemode.DT_Sparks_DL_Gamemode" "NewRow"))
  (return (Utilities|Struct|BreakSparksGameModeDataLayers _returnvalue)))

(fn GetPlaylistData
  (bind _returnvalue (Utilities|GetDataTableRow "/SparksMaps/Environment/BP/DataTables/DT_Sparks_DL_Playlists.DT_Sparks_DL_Playlists" "NewRow"))
  (return (Utilities|Struct|BreakSparksPlaylistDatalayers _returnvalue)))

(fn ConstructionScript)
