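# M_EnvPropBase is a masked Default Lit base material for environment props, with lawn chair textures as defaults.
# It exposes Diffuse, Masks, Surface (packed specular, metallic, roughness) and NormalMap texture parameters.
# Masks red blends in a Tint multiply over Diffuse, and green is labeled as opacity mask.
# The UseTintChannel2 static switch optionally layers Tint2 through the blue mask channel.
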Material /SparksMaps/Environment/Materials/M_EnvPropBase
graph {
 properties {bAutomaticallySetUsageInEditor=false, blendMode=BLEND_Masked, usedShadingModels=Default Lit}
 nodes [
  n1 = TextureSampleParameter2D() {
   group=Texture Maps
   parameterName=Diffuse
   sortPriority=2
   texture=/Game/Environments/Asteria/Props/Outdoor/Outdoor_LawnChairWebbed_A/Textures/T_Outdoor_LawnChairWebbed_A_A_D
  }
  n2 = TextureSampleParameter2D() {
   desc="(R) Tint Mask, (G) Opacity Mask"
   group=Texture Maps
   parameterName=Masks
   sortPriority=5
   texture=/Game/Environments/Asteria/Props/Outdoor/Outdoor_LawnChairWebbed_A/Textures/T_Outdoor_LawnChairWebbed_A_A_M
  }
  n3 = LinearInterpolate(
   A=@n1.RGB
   B=Multiply(A=@n1.RGB, B=VectorParameter() {group=Texture Maps, parameterName=Tint, sortPriority=0, defaultValue=<rgba 1 1 1 1>}.RGB)
   Alpha=@n2.R
  )
  n4 = VectorParameter() {group=Texture Maps, parameterName=Tint2, sortPriority=0, defaultValue=<rgba 1 1 1 1>}
  StaticSwitchParameter(True=LinearInterpolate(A=@n3, B=Multiply(A=@n1.RGB, B=@n4.RGB), Alpha=@n2.B), False=@n3) {parameterName=UseTintChannel2}
  TextureSampleParameter2D() {
   desc="(R) Specular, (G) Metallic, (B) Roughness"
   group=Texture Maps
   parameterName=Surface
   samplerType=SAMPLERTYPE_Masks
   sortPriority=4
   texture=/Game/Environments/Asteria/Props/Outdoor/Outdoor_LawnChairWebbed_A/Textures/T_Outdoor_LawnChairWebbed_A_A_S
  }
  TextureSampleParameter2D() {
   group=Texture Maps
   parameterName=NormalMap
   samplerType=SAMPLERTYPE_Normal
   sortPriority=3
   texture=/Game/Environments/Asteria/Props/Outdoor/Outdoor_LawnChairWebbed_A/Textures/T_Outdoor_LawnChairWebbed_A_A_N
  }
 ]
}
