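# Basic Default Lit environment master material for SparksMaps, flagged for use with Nanite.
# It exposes Diffuse, Normal and packed Masks texture parameters, with Masks channels labeled Specular, Metallic and Roughness.
# An optional static-switch tint multiplies Diffuse by a color and its alpha as intensity, then lerps from the original using Diffuse itself as alpha, so brighter areas take more tint.
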Material /SparksMaps/Environment/Materials/M_Sparks_EnvBasic
graph {
 properties {bUsedWithNanite=true, pixelDepthOffsetMode=PDOM_Legacy, usedShadingModels=Default Lit}
 nodes [
  n1 = TextureSampleParameter2D() {parameterName=Diffuse, samplerType=SAMPLERTYPE_Masks, texture=/Game/Global/Textures/Default/Blanks/T_MidgreyMask}
  n2 = VectorParameter() {group=Tint, parameterName=BaseColor_Tint/Intensity, sortPriority=37, defaultValue=<rgba 1 1 1 1>}
  n3 = Multiply(A=@n2.RGB, B=@n1.RGB)
  StaticSwitchParameter(True=LinearInterpolate(A=@n1.RGB, B=Multiply(A=@n3, B=@n2.A), Alpha=@n1.RGB), False=@n1.RGB) {group=Tint, parameterName=Enable tint, sortPriority=36}
  TextureSampleParameter2D() {parameterName=Normal, samplerType=SAMPLERTYPE_Normal, texture=&FlatNormalBlue}
  TextureSampleParameter2D() {
   parameterName=Masks
   samplerType=SAMPLERTYPE_Masks
   texture=&T_MidgreyMask
   channelNames {a=Unused, b=Roughness, g=Metallic, r=Specular}
  }
 ]
}
